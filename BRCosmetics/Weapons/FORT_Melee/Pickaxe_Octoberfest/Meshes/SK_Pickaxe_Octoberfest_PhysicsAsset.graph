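# Physics asset for the Octoberfest pickaxe skeletal mesh, a cosmetic melee weapon, giving it ragdoll-style collision and jointed parts.
# Four capsule bodies cover the handle, left and right sides, and bottom ribbon.
# All three attachments constrain directly to the handle: both sides get very tight swing with moderate twist, while the ribbon swings and twists more freely, suggesting secondary dangle motion.
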PhysicsAsset /BRCosmetics/Weapons/FORT_Melee/Pickaxe_Octoberfest/Meshes/SK_Pickaxe_Octoberfest_PhysicsAsset
bodies {
 handle [{shapeType=Capsule, length=64.5005, radius=15, center=<0.2752 -0.9214 -36.1612>}]
 ribbon_bottom [{shapeType=Capsule, length=9.7586, radius=1.5302, center=<0.0002 -4.2109 -1.2846>, rotation=<rot 0 0 73.0333>}]
 side_l [{shapeType=Capsule, length=23.3571, radius=6.6429, center=<0 20.9552 14.1693>, rotation=<rot 0 0 56.7466>}]
 side_r [{shapeType=Capsule, length=20.2703, radius=9.7297, center=<0 -13.6577 21.0134>, rotation=<rot 0 0 -33.1074>}]
}
constraints [
 {bone1Name=handle, bone2Name=side_l, swing1LimitDegrees=2, swing2LimitDegrees=2, twistLimitDegrees=30}
 {bone1Name=handle, bone2Name=side_r, swing1LimitDegrees=2, swing2LimitDegrees=2, twistLimitDegrees=30}
 {bone1Name=handle, bone2Name=ribbon_bottom, swing1LimitDegrees=16, swing2LimitDegrees=16, twistLimitDegrees=90}
]
